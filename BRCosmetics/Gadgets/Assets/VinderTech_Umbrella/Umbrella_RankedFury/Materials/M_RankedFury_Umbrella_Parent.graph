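# Parent material for the Ranked Fury VinderTech umbrella gadget, built on the 2023 character parent function chain and outputting material attributes.
# The PreFX result passes through a Ranked Fury grunge function borrowed from the backpack cosmetic, driven by the Tech Art Mask red channel.
# A shared ranked thin-film function, fed the Tech Art Mask, then layers on before PostFX.
# PostFX's M, pre-skinned position, distance blend and fresnel inputs all take PreFX's M output.
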Material /BRCosmetics/Gadgets/Assets/VinderTech_Umbrella/Umbrella_RankedFury/Materials/M_RankedFury_Umbrella_Parent
graph {
 properties {
  bUseMaterialAttributes=true
  bUsedWithInstancedStaticMeshes=true
  bUsedWithMeshDeformer=true
  bUsedWithNanite=true
  bUsedWithSkeletalMesh=true
  pixelDepthOffsetMode=PDOM_Legacy
  shadingModel=MSM_FromMaterialExpression
  usedShadingModels=Default Lit | Subsurface | Clear Coat | Subsurface Profile | Cloth
 }
 nodes [
  n1 = MaterialFunctionCall() {materialFunction=&MF_CharacterParent_2023_PreFX}
  n2 = ComponentMask(None=@n1.Tech Art Mask) {r=true}
  MaterialFunctionCall(
   Material=MaterialFunctionCall!(
    Material Attributes=MaterialFunctionCall!(MaterialAttributes=@n1.Result, TechArtMask=@n2) {materialFunction=/BRCosmetics/Accessories/FORT_Backpacks/Backpack_M_MED_RankedFury/Materials/MF_RankedFury_GrungeMaterial}.Result
    Alpha=@n1.Tech Art Mask
   ) {materialFunction=/Game/Global/Materials/Cosmetics/Ranked/MF_RankedThinFilm}.Result
   M=@n1.M
   Pre Skinned Local Position=@n1.M
   Distance Blend=@n1.M
   Fresnel=@n1.M
  ) {materialFunction=&MF_CharacterParent_2023_PostFX}
 ]
}
